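; Fortnite fire extinguisher pickup that gives its weapon item on interaction and can explode when damaged, whether loose or attached to a held weapon.
; Damage lights the fuse, replicates the hit for a smoking cue, and drops a held extinguisher as a bouncing projectile.
; Explosion logic plays land or water cues, applies Curie water in a radius, and fades a smoke volume through a timeline.
; ShouldDie returns false to preserve collision when shot from a player's hand.
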(event Custom|AllowInteractDelay
  ;[0.5686 1 0.6196] Begin Play with Curie Activated
  (Utilities|Time|SetTimerbyEvent (delegate Custom|AllowInteractDelay) 1.0)
  (Variables|Interaction|SetbAllowInteract true)
  (Variables|Default|SetCurieActivated (Curie|IsCurieActive)))

(event EventBeginPlay
  ;[0.5686 1 0.6196] Begin Play with Curie Activated
  (|CallParentFunction)
  (Variables|Interaction|SetbAllowInteract false)
  (Utilities|Time|SetTimerbyEvent (delegate Custom|AllowInteractDelay) 1.0)
  (Variables|Interaction|SetbAllowInteract true)
  (Variables|Default|SetCurieActivated (Curie|IsCurieActive)))

(event Interaction|EventBlueprintOnInteract
  (params (InteractingPawn :type Fort%20Pawn%20Object%20Reference) (InteractionBeingAttempted :type EInteractionBeingAttempted%20Enum :default "FirstInteraction"))
  ;[0.5686 1 0.6196] Interact
  (if (Utilities|Enum|Equal(Enum) (Fortnite|GetSubGame) "Athena")
    (bind _asfort_player_pawn (Utilities|Casting|CastToFortPlayerPawn InteractingPawn))
    (bind _output_get (Variables|Default|SetPlayerPawn _asfort_player_pawn))
    (bind _asfort_player_controller_athena (Utilities|Casting|CastToFortPlayerControllerAthena (Fortnite|GetFortPlayerControllerfromActor _output_get)))
    (Variables|Default|SetPlayerController _asfort_player_controller_athena)
    (Ability|ExecuteGameplayCueWithParamsReplicated (Variables|GameplayCues|GetGC_Pickup) (Ability|GameplayCue|MakeGameplayCueParameters :NormalizedMagnitude 0.0 :RawMagnitude 0.0 :Location (Transformation|GetActorLocation) :Normal (Transformation|GetActorForwardVector) :GameplayEffectLevel 1 :AbilityLevel (Math|Conversions|ToInteger(Byte) (Variables|GameplayCues|GetObjectWeights)) :bReplicateLocationWhenUsingMinimalRepProxy false))
    (if (Class|FortItemEntryComponent|HasItem (Variables|Default|GetFortItemEntry))
      (Class|FortItemEntryComponent|AddItemtoInventoryOwner (Variables|Default|GetFortItemEntry) (Variables|Default|GetPlayerController) false)
      (else
        (Inventory|GiveItemtoInventoryOwner (Variables|Default|GetPlayerController) (Variables|Default|GetWeaponItemDefinition))))
    (CallFunction|%28ServerOnly%29HideandDestroy)))

(event Collision|EventActorBeginOverlap
  (params (OtherActor :type Actor%20Object%20Reference))
  (|CallParentFunction OtherActor) ; This node is disabled and will not be called.\nDrag off pins to build functionality.
)

(event EventTick
  (params (DeltaSeconds :type Float%20%28single-precision%29))
  (|CallParentFunction DeltaSeconds) ; This node is disabled and will not be called.\nDrag off pins to build functionality.
)

(event Custom|ExplosionTrace
  ;[0.5188 1 0.7049] On Explode, send LastDamageInstigator event data to self.
  (Utilities|FlowControl|DoOnce false
    (:Completed
      (bind _outcomponents (Collision|SphereOverlapComponents (Transformation|GetActorLocation) (ScalableFloat|GetValueatLevel (Variables|Default|GetRadius)) (Utilities|Array|MakeArray "ObjectTypeQuery3" "ObjectTypeQuery5" "ObjectTypeQuery1" "ObjectTypeQuery2")))
      (for _ _outcomponents
        _))))

(event EventOnDamageServer
  (params (Damage :type Float%20%28single-precision%29) (DamageTags :type Gameplay%20Tag%20Container%20Structure%20%28by%20ref%29) (Momentum :type Vector :default "0, 0, 0") (HitInfo :type Hit%20Result%20Structure%20%28by%20ref%29) (InstigatedBy :type Controller%20Object%20Reference) (DamageCauser :type Actor%20Object%20Reference) (EffectContext :type Gameplay%20Effect%20Context%20Handle%20Structure))
  ;[1 1 1] On Damage Server
  (Utilities|FlowControl|DoOnce false
    (:Completed
      (bind _asfort_player_controller (Utilities|Casting|CastToFortPlayerController InstigatedBy))
      (Variables|Default|SetExplodeInstigator (Game|Player|GetPlayerPawn _asfort_player_controller))
      (Variables|Default|SetLastHit HitInfo) ; On Rep - calls Smoke GC
      (CallFunction|SetFuseandDrop))))

(event Custom|SetHeldItemInfo
  (params (InventoryInterface :type Fort%20Inventory%20Owner%20Interface%20Interface) (GUID :type Guid%20Structure) (WeaponAttachedTo :type Fort%20Weapon%20Object%20Reference))
  ;[1 1 1] Set Held Item Info
  (Variables|Default|SetHeldItemInventoryInterface InventoryInterface)
  (Variables|Default|SetHeldItemGUID GUID)
  (Variables|Default|SetWeaponAttachedTo WeaponAttachedTo)
  (Variables|Bools|SetIsAttachedToWeapon true))

(event Custom|DropHeldItem
  ;[1 1 1] Drop Held Item
  (Networking|FlushNetDormancy)
  (Variables|Bools|SetDropped true)
  (Transformation|DetachFromActor "KeepWorld" "KeepWorld" "KeepWorld")
  (Development|SetHiddeninGame (Variables|Default|GetFireExtinguisher) false false)
  (Collision|SetCollisionProfileName (Variables|Default|GetCapsule) "FortProjectileHitAllPawns" true :!class /Script/Engine.PrimitiveComponent)
  (Collision|SetCollisionEnabled (Variables|Default|GetCapsule) "QueryAndPhysics")
  (Projectile|SetVelocity (Variables|Default|GetFortProjectileMovement) (0.0 0.0 1.0))
  (bind _asb_fire_extinguisher_weapon (Utilities|Casting|CastToB_FireExtinguisherWeapon (Variables|Default|GetWeaponAttachedTo))
    (:then
      (Class|BFireExtinguisherWeapon|BGADestroyedWhileHeld _asb_fire_extinguisher_weapon))
    (:CastFailed
      _))
  (InventoryManagement|RemoveItem (Variables|Default|GetHeldItemInventoryInterface) (Variables|Default|GetHeldItemGUID)))

(event Custom|SetFuseAndDrop
  ;[1 1 1] Set Fuse and Drop
  (Variables|Bools|SetFuseLit true)
  (Networking|SwitchHasAuthority
    (:Authority
      (bind _returnvalue (Utilities|Time|SetTimerbyEvent (EventDispatchers|CreateEvent :!event Custom|ExplodeAndDie) (Variables|Emitter|GetFuseDuration)))
      (Variables|Default|SetFuseTimerHandle _returnvalue)
      (if (Variables|Bools|GetIsAttachedToWeapon)
        (CallFunction|DropHeldItem)
        (CallFunction|ExplodeandDie)))))

(event Custom|ExplodeAndDie
  ;[1 1 1] Explode and Die
  (if (Variables|Bools|GetShouldExplode)
    (Actor|DestroyActor)
    (CallFunction|HideandDestroy)
    (GameplayCue|ExecuteGameplayCue%28Burst%29Local self (select (Water|IsinWater (Variables|Default|GetFortWaterInteraction)) (Variables|GameplayCues|GetExplosionCUE_Water) (Variables|GameplayCues|GetExplosionCUE)) (Ability|GameplayCue|MakeGameplayCueParameters 0.0))
    (Utilities|FlowControl|DoOnce false
      (:Completed
        (CallFunction|SplashWater)
        (Utilities|Time|SetTimerbyFunctionName 0 "AfterSmoke" (Variables|AfterSmoke|GetAfterSmokeDelay))))))

(event Custom|AfterSmoke
  ;[1 1 1] After explosion smoke
  (Networking|CanExecuteCosmeticEvents
    (:True
      (Rendering|ToggleVisibility (Variables|Default|GetSmokeVolumeMesh) false)
      (Variables|AfterSmoke|SetSmokeLifespan (Math|Float|float+float (Variables|AfterSmoke|GetSmokeLifespan) (Variables|AfterSmoke|GetSmokeDelayBeforeFadeIn) (Variables|AfterSmoke|GetSmokeFadeOutTime)))
      (Actor|SetLifeSpan (Variables|AfterSmoke|GetSmokeLifespan))
      (Utilities|Time|SetTimerbyEvent (EventDispatchers|CreateEvent :!event Custom|SmokeFadeOut) (Utilities|ClampMinFloat (- (Variables|AfterSmoke|GetSmokeLifespan) (Variables|AfterSmoke|GetSmokeFadeOutTime)) (Variables|AfterSmoke|GetSmokeDelayBeforeFadeIn)))
      (Utilities|Time|SetTimerbyEvent (EventDispatchers|CreateEvent :!event Custom|SmokeFadeIn) (Variables|AfterSmoke|GetSmokeDelayBeforeFadeIn))
      (Development|SetHiddeninGame (Variables|Default|GetSmokeVolumeMesh) false false))
    (:False
      (Actor|DestroyActor))))

(event Custom|HideAndDestroy
  ;[1 1 1] Hide and Destroy
  (Development|SetHiddeninGame (Variables|Default|GetFireExtinguisher) true false)
  (Actor|DestroyActor))

(event Custom|EventDestroyed
  ;[1 1 1] Destroyed String
  (Development|PrintString "BGA Fire Extinguisher Destroyed"))

(event Custom|SplashWater
  ;[0.3859 0.5206 1] Splash Water puts that fire out!
  (Variables|Curie|SetHitLocation (Transformation|GetActorLocation))
  (Utilities|FlowControl|DoOnce false
    (:Completed
      (CallFunction|ApplyWaterinRadiusifCurieEnabled (Variables|Default|GetLastHit)))))

(event Custom|%28ServerOnly%29%20HideAndDestroy
  ;[1 1 1] Hide and Destroy
  (Development|SetHiddeninGame (Variables|Default|GetFireExtinguisher) true false)
  (Actor|DestroyActor))

(event OnProjectileBounce(FortProjectileMovement)
  (params (ImpactResult :type Hit%20Result%20Structure%20%28by%20ref%29) (ImpactVelocity :type Vector%20%28by%20ref%29 :default "0, 0, 0"))
  (bind (_bblockinghit _binitialoverlap _time _distance _location _impactpoint _normal _impactnormal _physmat _hitactor _hitcomponent) (Collision|BreakHitResult ImpactResult))
  ;[1 1 1] On bounce, if hit a no stick object, adjust velocity to bounce away
  (if (or (Actor|ActorHasTag _hitactor "None") (Components|ComponentHasTag _hitcomponent "None"))
    (Projectile|SetVelocity (Variables|Default|GetFortProjectileMovement) (* (Math|Vector|Normalize (Math|Vector|vector+vector (Math|Vector|MirrorVectorbyNormal (* (Variables|/Script/Engine.MovementComponent|GetVelocity (Variables|Default|GetFortProjectileMovement)) 0.75) "0, 0,1.000000") (Math|Vector|MakeVector 0.0 0.0 (* (Math|Vector|VectorLength (Variables|/Script/Engine.MovementComponent|GetVelocity (Variables|Default|GetFortProjectileMovement))) 0.25)))) (Math|Float|Max(Float) (Math|Vector|VectorLength (Math|Vector|vector+vector (Math|Vector|MirrorVectorbyNormal (* (Variables|/Script/Engine.MovementComponent|GetVelocity (Variables|Default|GetFortProjectileMovement)) 0.75) "0, 0,1.000000") (Math|Vector|MakeVector 0.0 0.0 (* (Math|Vector|VectorLength (Variables|/Script/Engine.MovementComponent|GetVelocity (Variables|Default|GetFortProjectileMovement))) 0.25)))) (Variables|Default|GetMinSpeedOfNoStickBounce))))))

(event Custom|SmokeFadeIn
  ;[1 1 1] After Smoke Timeline
  (Components|Timeline|SetPlayRate (Variables|BGA_FireExtinguisher_Pickup|GetFadeSmokeTL) (Utilities|Reciprocal (Variables|AfterSmoke|GetSmokeFadeInTime)))
  (bind (_direction _float) (Timeline|FadeSmokeTL PlayFromStart)
    (:Update
      (bind _output_get (Variables|AfterSmoke|SetSmokeFadeValue _float))
      (Rendering|Material|SetScalarParameterValue (Variables|AfterSmoke|GetSmokeVolumeMeshMID) "Fade" _output_get :!class /Script/Engine.MaterialInstanceDynamic)
      (Rendering|Material|SetScalarParameterValue (Variables|AfterSmoke|GetSmokeScreenMID) "Opacity" _output_get :!class /Script/Engine.MaterialInstanceDynamic)
      (Utilities|FlowControl|DoOnce false
        (:Completed
          (Audio|Components|Audio|SetVolumeMultiplier (Variables|Default|GetAudio_ReleaseSmoke) (Variables|AfterSmoke|GetSmokeFadeValue) :!class /Script/Engine.AudioComponent))))
    (:Finished
      (switch Utilities|FlowControl|Switch|SwitchonETimelineDirection _direction
        (:Forward
          (Components|Activation|Activate (Variables|Default|GetNS_Athena_SmokeGrenade_OuterSmoke_01) false))))))

(event Custom|SmokeFadeOut
  (bind _ns_athena_smokegrenade_outersmoke_01 (Variables|Default|GetNS_Athena_SmokeGrenade_OuterSmoke_01))
  ;[1 1 1] After Smoke Timeline
  (Components|Timeline|SetPlayRate (Variables|BGA_FireExtinguisher_Pickup|GetFadeSmokeTL) (Utilities|Reciprocal (Variables|AfterSmoke|GetSmokeFadeOutTime)))
  (Components|Activation|Deactivate _ns_athena_smokegrenade_outersmoke_01)
  (Timeline|FadeSmokeTL ReverseFromEnd))

(event EventOnDeathServer
  (params (Damage :type Float%20%28single-precision%29) (DamageTags :type Gameplay%20Tag%20Container%20Structure%20%28by%20ref%29) (Momentum :type Vector :default "0, 0, 0") (HitInfo :type Hit%20Result%20Structure%20%28by%20ref%29) (InstigatedBy :type Controller%20Object%20Reference) (DamageCauser :type Actor%20Object%20Reference) (EffectContext :type Gameplay%20Effect%20Context%20Handle%20Structure))
  ;[1 1 1] Override on server death function to get rid of parent death fx
  (Rendering|SetActorHiddenInGame false)
  (Utilities|IsValid (Variables|Default|GetWeaponAttachedTo)
    (:Is%20Valid
      (bind _asb_fire_extinguisher_weapon (Utilities|Casting|CastToB_FireExtinguisherWeapon (Variables|Default|GetWeaponAttachedTo)))
      (Class|BFireExtinguisherWeapon|BGADestroyedWhileHeld _asb_fire_extinguisher_weapon)
      (Development|PrintString "Died" true true "(R=1.000000,G=0.000000,B=0.023606,A=1.000000)" 5.0))))

(fn ApplyWaterInRadiusIfCurieEnabled
  (params (HitResult :type Hit%20Result%20Structure))
  (Networking|CanExecuteCosmeticEvents
    (:False
      (if (Curie|IsCurieActive)
        (bind _outactors (Collision|SphereOverlapActors (Variables|Curie|GetHitLocation) (* (Variables|Curie|GetWaterTraceRadius) (Math|Vector|VectorLength (Transformation|GetActorScale3D))) (Variables|Curie|GetWaterTraceObjectTypes) 0 (Utilities|Array|MakeArray self)))
        (for _array_element _outactors
          (Utilities|IsValid _array_element
            (:Is%20Valid
              (Curie|ExecuteWaterApplication (Variables|Default|GetLastDamageInstigator) self _array_element 1e+02 (Collision|MakeHitResult false false 0.0 0.0 (Variables|Curie|GetHitLocation) (Variables|Curie|GetHitLocation) "0,0,1" "0,0,1" 0 _array_element) (Variables|Curie|GetLandscapeExtinguishRadius)))))))))

(fn BlueprintCanInteract
  (params (InteractingPawn :type Fort%20Pawn%20Object%20Reference) (InteractionBeingAttempted :type EInteractionBeingAttempted%20Enum :default "FirstInteraction") (InteractionType :type TInteractionType%20Enum :default "IT_NoInteraction"))
  (return (not (Variables|Bools|GetFuseLit))))

(fn BlueprintGetInteractionString
  (params (InteractingPawn :type Fort%20Pawn%20Object%20Reference) (InteractionBeingAttempted :type EInteractionBeingAttempted%20Enum :default "FirstInteraction"))
  (return "NSLOCTEXT(\"[4CC246F04BC83A11BDA9D9BD113658CD]\", \"91F4A2794D81253C6E03C7A0B78CC383\", \"Pickup Fire Extinguisher\")"))

(fn BlueprintGetInteractionTime
  (params (InteractingPawn :type Fort%20Pawn%20Object%20Reference) (InteractionBeingAttempted :type EInteractionBeingAttempted%20Enum :default "FirstInteraction"))
  (return 0.25 false))

(fn BlueprintOnLocalInteract
  (params (InteractingPawn :type Fort%20Player%20Pawn%20Object%20Reference) (InteractionBeingAttempted :type EInteractionBeingAttempted%20Enum :default "FirstInteraction"))
  ;[1 1 1] Prevent tplayer from attempting to pick up Fire Extinguisher if inventory is full (and player is holding pickaxe so the pickup won't replace another slotted item).
  (if (> (Inventory|CountInventoryOverflowfromAddingItem (Fortnite|GetFortPlayerControllerfromActor InteractingPawn) (Loot|CreateItemEntry (Variables|Default|GetWeaponItemDefinition)) false true) 0)
    (FailedAction|UnabletoPerformAction InteractingPawn true (Variables|Default|GetTag_FailedInteract_InventoryIsFull))
    (Audio|PlaySoundatLocation (Variables|Default|GetFailedInteractSound) (Transformation|GetActorLocation))
    (return true)
    (else
      (return false))))

(fn OnRep_LastHit
  (bind _self self)
  (Networking|CanExecuteCosmeticEvents
    (:True
      (bind _outcueparameters (GameplayCue|BuildCueParametersfromHitResult _self (Variables|Default|GetLastHit)))
      (Projectile|AddGameplayCue%28Looping%29Local _self (Variables|Default|GetGC_Smoking) _outcueparameters))))

(fn ShouldDie
  (params (Damage :type Float%20%28single-precision%29) (EventInstigator :type Controller%20Object%20Reference) (DamageCauser :type Actor%20Object%20Reference) (bIsTest :type Boolean) (EffectContext :type Gameplay%20Effect%20Context%20Handle%20Structure%20%28by%20ref%29))
  ;[1 1 1] This is set to false so that we don't lose collision when this is shot out of a player's hand - call Hide and Destroy to finish killing the object
  (return false))

(fn ConstructionScript
  (Networking|CanExecuteCosmeticEvents
    (:True
      (Transformation|SetWorldScale3D (Variables|Default|GetSmokeVolumeMesh) (Math|Conversions|ToVector(Float) (Variables|AfterSmoke|GetSmokeRadius)))
      (bind _returnvalue (Rendering|Material|CreateDynamicMaterialInstance (Variables|Default|GetSmokeVolumeMesh) 0 "/Game/Athena/Items/Consumables/SmokeGrenade/M_Athena_SmokeGrenade_Outer_01.M_Athena_SmokeGrenade_Outer_01" "None" :!class /Script/Engine.PrimitiveComponent))
      (bind _output_get (Variables|AfterSmoke|SetSmokeVolumeMeshMID _returnvalue))
      (Rendering|Material|SetScalarParameterValue _output_get "Fade" 1.0 :!class /Script/Engine.MaterialInstanceDynamic)
      (Rendering|Material|SetScalarParameterValue _output_get "Radius" (Variables|AfterSmoke|GetSmokeRadius) :!class /Script/Engine.MaterialInstanceDynamic)
      (bind _returnvalue_2 (Rendering|Material|CreateDynamicMaterialInstance (Variables|Default|GetSmokeScreen_Sphere) 0 "/Game/Athena/Items/Consumables/SmokeGrenade/M_SmokeGrenadeSphere.M_SmokeGrenadeSphere" "None" :!class /Script/Engine.PrimitiveComponent))
      (bind _output_get_2 (Variables|AfterSmoke|SetSmokeScreenMID _returnvalue_2))
      (Rendering|Material|SetScalarParameterValue _output_get_2 "Opacity" 0.0 :!class /Script/Engine.MaterialInstanceDynamic)
      (Rendering|Material|SetScalarParameterValue _output_get_2 "Radius" (* (Variables|AfterSmoke|GetSmokeRadius) 1.63) :!class /Script/Engine.MaterialInstanceDynamic))))
