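; Fire extinguisher weapon Blueprint that, when IsShootableWhileHeld is set, carries a hidden pickup actor so the extinguisher can be shot in hand.
; Server equip optionally adds ammo, spawns the pickup and passes it inventory info; UpdateAndAttachBGA, also run from OnRep, attaches it with weapons-only trigger collision.
; Unequip or holster hides and destroys an undropped pickup; remote-side holster handlers forward to Replicate events, with unholster respawning it.
; Optional HUD key prompts follow targeting state.
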(event EventBeginPlay
  ;[1 1 1] On Begin Play set up HUD actions
  (|CallParentFunction)
  (if (Variables|Default|GetShowHUDKeyActions)
    (WeaponHUDData|SetHudKeyActionsVisibility (Utilities|Array|MakeArray (Utilities|Struct|MakeWeaponHudKeyActionVisibility 0 true) (Utilities|Struct|MakeWeaponHudKeyActionVisibility 1 true)))))

(event Custom|HUDKeyAction_Targeting
  (params (IsTargeting :type Boolean))
  ;[1 1 1] Toggle HUD key actions based on targeting - called from secondary fire GA
  (if (Variables|Default|GetShowHUDKeyActions)
    (if IsTargeting
      (WeaponHUDData|SetHudKeyActionsVisibility (Utilities|Array|MakeArray (Utilities|Struct|MakeWeaponHudKeyActionVisibility 1) (Utilities|Struct|MakeWeaponHudKeyActionVisibility 2 true)))
      (else
        (WeaponHUDData|SetHudKeyActionsVisibility (Utilities|Array|MakeArray (Utilities|Struct|MakeWeaponHudKeyActionVisibility 1 true) (Utilities|Struct|MakeWeaponHudKeyActionVisibility 2)))))))

(event Custom|UpdateAndAttachBGA
  (bind _attachedbgaactor (Variables|Default|GetAttachedBGAActor))
  (bind _attachedbgaactor_2 (Variables|Default|GetAttachedBGAActor))
  ;[1 1 1] Update and attach BGA
  (Utilities|IsValid _attachedbgaactor
    (:Is%20Valid
      (Game|Components|ProjectileMovement|StopSimulating (Variables|/Game/Athena/Items/HeldObjects/HeldObjectVersion2/BGA_HeldObject_Parent.BGA_HeldObject_Parent_C|GetFortProjectileMovement (Variables|Default|GetAttachedBGAActor)) (Collision|MakeHitResult false false 0.0 0.0 (Transformation|GetActorLocation self)))
      (Collision|SetIgnoreInstigatorCollision (Variables|Default|GetAttachedBGAActor) true :!class /Script/FortniteGame.BuildingGameplayActor)
      (Collision|SetCollisionEnabled (Variables|/Game/Creative/Test/Props/FireExtinguisherSpray/BGA_FireExtinguisher_Pickup.BGA_FireExtinguisher_Pickup_C|GetFireExtinguisher _attachedbgaactor_2) "NoCollision")
      (Collision|SetCollisionProfileName (Variables|/Game/Athena/Items/HeldObjects/HeldObjectVersion2/BGA_HeldObject_Parent.BGA_HeldObject_Parent_C|GetCapsule _attachedbgaactor_2) "FortTriggerOnlyWeapons" true :!class /Script/Engine.PrimitiveComponent)
      (Development|SetHiddeninGame (Variables|/Game/Creative/Test/Props/FireExtinguisherSpray/BGA_FireExtinguisher_Pickup.BGA_FireExtinguisher_Pickup_C|GetFireExtinguisher (Variables|Default|GetAttachedBGAActor)) true false)
      (Transformation|AttachActorToActor (Variables|Default|GetAttachedBGAActor) self (Variables|Default|GetAttachSocket) "SnapToTarget" "SnapToTarget" "SnapToTarget" true)
      (Transformation|SetActorRelativeTransform (Variables|Default|GetAttachedBGAActor) (Variables|Default|GetAttachedBGARelativeTransform) false true)
      (Variables|/Game/Athena/Items/HeldObjects/HeldObjectVersion2/BGA_HeldObject_Parent.BGA_HeldObject_Parent_C|SetRepStartMoving 0 (Variables|Default|GetAttachedBGAActor)))
    (:Is%20Not%20Valid
      (Actor|DestroyActor _attachedbgaactor))))

(event FortWeapon|EventOnEquippedAndAttached
  (bind _returnvalue (Game|GetInstigator))
  ;[1 1 1] On Equip
  (|CallParentFunction)
  (CallFunction|BindHolsterEvents)
  (Networking|SwitchHasAuthority
    (:Authority
      (if (Variables|Default|GetEnableAmmoMod)
        (FortWeapon|ModAmmo 5)
        (else
          _))
      (if (> (ScalableFloat|GetValueatLevel (Variables|Default|GetIsShootableWhileHeld)) 0.0)
        (bind _returnvalue_2 (Building|SpawnBuildingGameplayActor "/Game/Creative/Test/Props/FireExtinguisherSpray/BGA_FireExtinguisher_Pickup.BGA_FireExtinguisher_Pickup_C" (Transformation|GetActorTransform self) (Game|GetInstigator)))
        (bind _asbga_fire_extinguisher_pickup (Utilities|Casting|CastToBGA_FireExtinguisher_Pickup _returnvalue_2)
          (:then
            (Variables|Default|SetAttachedBGAActor _asbga_fire_extinguisher_pickup)
            (CallFunction|UpdateandAttachBGA)
            ;[1 1 1] Set Held Item Info
            (bind _asfort_player_pawn (Utilities|Casting|CastToFortPlayerPawn _returnvalue)
              (:then
                (Variables|Default|SetInventoryGUID (FortWeapon|GetInventoryGUID (Variables|/Script/FortniteGame.FortPawn|GetCurrentWeapon _asfort_player_pawn)))
                (Variables|Default|SetWeaponReference (Variables|/Script/FortniteGame.FortPawn|GetCurrentWeapon _asfort_player_pawn))
                (bind _asfort_inventory_owner_interface (Utilities|Casting|CastToFortInventoryOwnerInterface (Pawn|GetController _asfort_player_pawn))
                  (:then
                    (Variables|Default|SetInventoryInterface _asfort_inventory_owner_interface)
                    (Class|BGAFireExtinguisherPickup|SetHeldItemInfo (Variables|Default|GetAttachedBGAActor) (Variables|Default|GetInventoryInterface) (Variables|Default|GetInventoryGUID) (Variables|Default|GetWeaponReference)))
                  (:CastFailed
                    _)))
              (:CastFailed
                (bind _asfort_aipawn (Utilities|Casting|CastToFortAIPawn _returnvalue)
                  (:then
                    (Variables|Default|SetInventoryGUID (FortWeapon|GetInventoryGUID (Variables|/Script/FortniteGame.FortPawn|GetCurrentWeapon _asfort_aipawn)))
                    (Variables|Default|SetWeaponReference (Variables|/Script/FortniteGame.FortPawn|GetCurrentWeapon _asfort_aipawn))
                    (bind _asfort_inventory_owner_interface_2 (Utilities|Casting|CastToFortInventoryOwnerInterface _asfort_aipawn)
                      (:then
                        (Variables|Default|SetInventoryInterface _asfort_inventory_owner_interface_2))
                      (:CastFailed
                        _)))
                  (:CastFailed
                    (Actor|DestroyActor (Variables|Default|GetAttachedBGAActor))))))
            (Actor|DestroyActor (Variables|Default|GetAttachedBGAActor)))
          (:CastFailed
            (Actor|DestroyActor _returnvalue_2)))))))

(event Custom|ReplicateOnUnholstered
  (bind _returnvalue (Game|GetInstigator))
  ;[1 1 1] On Equip
  (Networking|SwitchHasAuthority
    (:Authority
      (if (Variables|Default|GetEnableAmmoMod)
        (FortWeapon|ModAmmo 5)
        (else
          _))
      (if (> (ScalableFloat|GetValueatLevel (Variables|Default|GetIsShootableWhileHeld)) 0.0)
        (bind _returnvalue_2 (Building|SpawnBuildingGameplayActor "/Game/Creative/Test/Props/FireExtinguisherSpray/BGA_FireExtinguisher_Pickup.BGA_FireExtinguisher_Pickup_C" (Transformation|GetActorTransform self) (Game|GetInstigator)))
        (bind _asbga_fire_extinguisher_pickup (Utilities|Casting|CastToBGA_FireExtinguisher_Pickup _returnvalue_2)
          (:then
            (Variables|Default|SetAttachedBGAActor _asbga_fire_extinguisher_pickup)
            (CallFunction|UpdateandAttachBGA)
            ;[1 1 1] Set Held Item Info
            (bind _asfort_player_pawn (Utilities|Casting|CastToFortPlayerPawn _returnvalue)
              (:then
                (Variables|Default|SetInventoryGUID (FortWeapon|GetInventoryGUID (Variables|/Script/FortniteGame.FortPawn|GetCurrentWeapon _asfort_player_pawn)))
                (Variables|Default|SetWeaponReference (Variables|/Script/FortniteGame.FortPawn|GetCurrentWeapon _asfort_player_pawn))
                (bind _asfort_inventory_owner_interface (Utilities|Casting|CastToFortInventoryOwnerInterface (Pawn|GetController _asfort_player_pawn))
                  (:then
                    (Variables|Default|SetInventoryInterface _asfort_inventory_owner_interface)
                    (Class|BGAFireExtinguisherPickup|SetHeldItemInfo (Variables|Default|GetAttachedBGAActor) (Variables|Default|GetInventoryInterface) (Variables|Default|GetInventoryGUID) (Variables|Default|GetWeaponReference)))
                  (:CastFailed
                    _)))
              (:CastFailed
                (bind _asfort_aipawn (Utilities|Casting|CastToFortAIPawn _returnvalue)
                  (:then
                    (Variables|Default|SetInventoryGUID (FortWeapon|GetInventoryGUID (Variables|/Script/FortniteGame.FortPawn|GetCurrentWeapon _asfort_aipawn)))
                    (Variables|Default|SetWeaponReference (Variables|/Script/FortniteGame.FortPawn|GetCurrentWeapon _asfort_aipawn))
                    (bind _asfort_inventory_owner_interface_2 (Utilities|Casting|CastToFortInventoryOwnerInterface _asfort_aipawn)
                      (:then
                        (Variables|Default|SetInventoryInterface _asfort_inventory_owner_interface_2))
                      (:CastFailed
                        _)))
                  (:CastFailed
                    (Actor|DestroyActor (Variables|Default|GetAttachedBGAActor))))))
            (Actor|DestroyActor (Variables|Default|GetAttachedBGAActor)))
          (:CastFailed
            (Actor|DestroyActor _returnvalue_2)))))))

(event Custom|OnUnholstered
  ;[1 1 1] On Equip
  (Networking|SwitchHasAuthority
    (:Remote
      (CallFunction|ReplicateonUnholstered))))

(event FortWeapon|EventOnLocalUnEquip
  ;[1 1 1] On Unequip
  (|CallParentFunction)
  (CallFunction|UnbindHolsterEvents)
  (Utilities|IsValid (Variables|Default|GetAttachedBGAActor)
    (:Is%20Valid
      (if (not (Variables|/Game/Creative/Test/Props/FireExtinguisherSpray/BGA_FireExtinguisher_Pickup.BGA_FireExtinguisher_Pickup_C|GetDropped (Variables|Default|GetAttachedBGAActor)))
        (Variables|/Game/Creative/Test/Props/FireExtinguisherSpray/BGA_FireExtinguisher_Pickup.BGA_FireExtinguisher_Pickup_C|SetShouldExplode true (Variables|Default|GetAttachedBGAActor))
        (Class|BGAFireExtinguisherPickup|HideandDestroy (Variables|Default|GetAttachedBGAActor))))))

(event Custom|ReplicateOnHolstered
  ;[1 1 1] On Unequip
  (Utilities|IsValid (Variables|Default|GetAttachedBGAActor)
    (:Is%20Valid
      (if (not (Variables|/Game/Creative/Test/Props/FireExtinguisherSpray/BGA_FireExtinguisher_Pickup.BGA_FireExtinguisher_Pickup_C|GetDropped (Variables|Default|GetAttachedBGAActor)))
        (Variables|/Game/Creative/Test/Props/FireExtinguisherSpray/BGA_FireExtinguisher_Pickup.BGA_FireExtinguisher_Pickup_C|SetShouldExplode true (Variables|Default|GetAttachedBGAActor))
        (Class|BGAFireExtinguisherPickup|HideandDestroy (Variables|Default|GetAttachedBGAActor))))))

(event Custom|OnHolstered
  ;[1 1 1] On Unequip
  (Networking|SwitchHasAuthority
    (:Remote
      (CallFunction|ReplicateonHolstered)
      (Utilities|IsValid (Variables|Default|GetAttachedBGAActor)
        (:Is%20Valid
          (if (not (Variables|/Game/Creative/Test/Props/FireExtinguisherSpray/BGA_FireExtinguisher_Pickup.BGA_FireExtinguisher_Pickup_C|GetDropped (Variables|Default|GetAttachedBGAActor)))
            (Variables|/Game/Creative/Test/Props/FireExtinguisherSpray/BGA_FireExtinguisher_Pickup.BGA_FireExtinguisher_Pickup_C|SetShouldExplode true (Variables|Default|GetAttachedBGAActor))
            (Class|BGAFireExtinguisherPickup|HideandDestroy (Variables|Default|GetAttachedBGAActor))))))))

(event Custom|BindHolsterEvents
  ;[1 1 1] Holster Event binding / unbinding
  (bind _asfort_pawn (Utilities|Casting|CastToFortPawn (Game|GetInstigator)))
  (EventDispatchers|BindEventtoOnHolsteredEvent _asfort_pawn (EventDispatchers|CreateEvent :!event Custom|OnHolstered))
  (EventDispatchers|BindEventtoOnUnholsteredEvent _asfort_pawn (EventDispatchers|CreateEvent :!event Custom|OnUnholstered)))

(event Custom|BGADestroyedWhileHeld
  ;[1 1 1] On BGA shot out of hands
  (CallFunction|UnbindHolsterEvents))

(event Custom|UnbindHolsterEvents
  ;[1 1 1] Holster Event binding / unbinding
  (bind _asfort_pawn (Utilities|Casting|CastToFortPawn (Game|GetInstigator)))
  (EventDispatchers|UnbindEventfromOnHolsteredEvent _asfort_pawn (EventDispatchers|CreateEvent :!event Custom|OnHolstered))
  (EventDispatchers|UnbindEventfromOnUnholsteredEvent _asfort_pawn (EventDispatchers|CreateEvent :!event Custom|OnUnholstered)))

(fn OnRep_AttachedBGAActor
  (CallFunction|UpdateandAttachBGA))

(fn ConstructionScript)
