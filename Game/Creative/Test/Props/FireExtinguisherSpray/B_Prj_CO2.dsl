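; CO2 projectile for the fire extinguisher spray that applies water on impact through the Curie system.
; The first hit stores location and normal, applies water, prints a debug message and triggers an explosion at the actor.
; Water application runs only where cosmetic events cannot execute and Curie is active, hitting actors in a scale-adjusted sphere overlap.
; The capsule drives projectile movement, and the static mesh hides on stop.
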(event Collision|EventHit
  (params (MyComp :type Primitive%20Component%20Object%20Reference) (Other :type Actor%20Object%20Reference) (OtherComp :type Primitive%20Component%20Object%20Reference) (bSelfMoved :type Boolean) (HitLocation :type Vector :default "0, 0, 0") (HitNormal :type Vector :default "0, 0, 0") (NormalImpulse :type Vector :default "0, 0, 0") (Hit :type Hit%20Result%20Structure%20%28by%20ref%29))
  ;[1 1 1] On hit, apply water.
  (Variables|Curie|SetHitLocation HitLocation)
  (Variables|Curie|SetHitNormal HitNormal)
  (Utilities|FlowControl|DoOnce false
    (:Completed
      (CallFunction|ApplyWaterinRadiusifCurieEnabled Hit)
      (Development|PrintString (Utilities|String|Append "Water hit: " (Utilities|GetDisplayName Other)) true true "(R=1.000000,G=0.432751,B=0.959595,A=1.000000)")
      (Delivery|DoExplosionatLocation (Transformation|GetActorLocation self)))))

(event EventBeginPlay
  ;[1 1 1] Begin Play
  (|CallParentFunction))

(event EventOnStop
  (params (Hit :type Hit%20Result%20Structure%20%28by%20ref%29))
  ;[1 1 1] On Stop Change Visibility
  (|CallParentFunction Hit)
  (Rendering|SetVisibility (Variables|Default|GetStaticMesh) false true))

(event EventDestroyed
  ;[1 1 1] On Destroyed
  (|CallParentFunction))

(fn ApplyWaterInRadiusIfCurieEnabled
  (params (HitResult :type Hit%20Result%20Structure))
  (Networking|CanExecuteCosmeticEvents
    (:False
      (if (Curie|IsCurieActive)
        (bind _outactors (Collision|SphereOverlapActors (Variables|Curie|GetHitLocation) (* (Variables|Curie|GetWaterTraceRadius) (Math|Vector|VectorLength (Transformation|GetActorScale3D))) (Variables|Curie|GetWaterTraceObjectTypes) 0 (Utilities|Array|MakeArray self)))
        (for _array_element _outactors
          (Utilities|IsValid _array_element
            (:Is%20Valid
              (Curie|ExecuteWaterApplication (Variables|Actor|GetInstigator) self _array_element 1e+02 (Collision|MakeHitResult false true 0.0 0.0 (Variables|Curie|GetHitLocation) (Variables|Curie|GetHitLocation) "0,0,1" "0,0,1" 0 _array_element) (Variables|Curie|GetLandscapeExtinguishRadius)))))))))

(fn ConstructionScript
  (|CallParentFunction)
  (Components|Movement|SetUpdatedComponent (Variables|Movement|GetProjectileMovementComponent) (Variables|Collision|GetCapsuleComponent)))
